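# Two-sided, lit effect material for the 5SpokeTigerSE cosmetic wheel, adding animated emissive tire patterns and supporting skeletal meshes.
# The tire texture's green channel masks a blend between CustomColor1 and CustomColor2, driven by a time-scrolled sine along U.
# Red minus green adds CustomColor3 lines, and blue adds an untinted glowing outline.
# Vertex color red blends this emissive layer over MF_WheelParent_Simple using material attributes.
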Material /VehicleCosmetics/Wheels/5SpokeTigerSE/Materials/M_Wheel_5SpokeTigerSE_Fx
graph {
 properties {
  bAutomaticallySetUsageInEditor=false
  bUseMaterialAttributes=true
  bUsedWithMeshDeformer=true
  bUsedWithSkeletalMesh=true
  twoSided=true
  usedShadingModels=Default Lit
 }
 nodes [
  n1 = TextureSample() {texture=/VehicleCosmetics/Wheels/5SpokeTigerSE/Textures/T_Tire_5SpokeTigerSE_RGB}
  n2 = Multiply(
   A=Multiply(A=Subtract(A=@n1.R, B=@n1.G), B=VectorParameter() {parameterName=CustomColor3, defaultValue=<rgba 0.9131 0.8632 0.1746 1>}.RGB)
   B=ScalarParameter() {defaultValue=2, parameterName=TireLineEmissive}
  ) {constB=2}
  n3 = VectorParameter() {parameterName=CustomColor2, defaultValue=<rgba 0.006 0.1195 0.491 1>}
  n4 = Multiply(A=Time(), B=Multiply(B=ScalarParameter() {defaultValue=0.2, group=3_InputProperties, parameterName=TurningSpeed}) {constA=-1})
  n5 = ComponentMask(None=TextureCoordinate() {uTiling=0.3333, vTiling=0.3333}) {r=true}
  n6 = Add(A=Sine(None=Add(A=@n4, B=@n5)))
  n7 = ScalarParameter() {defaultValue=0.5, parameterName=Tiling}
  n8 = Clamp(None=Power(Base=Multiply(A=@n6, B=@n7) {constB=3}))
  n9 = VectorParameter() {parameterName=CustomColor1, defaultValue=<rgba 0.5089 0.3613 0.0666 1>}
  n10 = Add(
   A=Multiply(A=@n1.G, B=LinearInterpolate(A=@n9.RGB, B=@n3.RGB, Alpha=@n8))
   B=Multiply(
    A=TextureSample() {texture=&T_Tire_5SpokeTigerSE_RGB}.B
    B=ScalarParameter() {defaultValue=5, parameterName=Tire_OutLineEmissive}
   )
  )
  BlendMaterialAttributes(
   A=MaterialFunctionCall() {materialFunction=&MF_WheelParent_Simple}.MaterialAttributes
   B=SetMaterialAttributes(Emissive Color=Add(A=@n2, B=@n10)) {attributeSetTypes [#1]}
   Alpha=VertexColor().R
  )
 ]
}
